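# Niagara glow effect for the Ornamental Soldier hat, drawing a faint, warm yellow pulsing sprite with a matching point light.
# A single emitter bursts an offset particle that does not expire with its lifetime.
# Waveforms drive both sprite scale and light intensity for a flicker or pulse.
# The User.bisHeadHidden bool sets the particle VisibilityTag to 1 or 0, apparently to switch the effect with head visibility.
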NiagaraSystem /BRCosmetics/Weapons/FORT_Melee/Effects/HolidayTime/NS_OrnamentalSoldier_Hat
dependencies {usedDataInterfaces [&NiagaraDataInterfaceCurve], usedDynamicInputs [], usedModules [&SystemState]}
emitters {Minimal {
  emitterUpdateScript [
   &EmitterState()
   &SpawnBurst_Instantaneous(Use Spawn Probability=false, UseLoopCountLimit=true)
  ]
  particleSpawnScript [&InitializeParticle(
    Color=<rgba 1 0.6333 0 0.025>
    Lifetime=5
    Position Offset=<5 0 0>
    Sprite Size Mode=::NewEnumerator3
    Uniform Sprite Size=50
    UsePositionOffset=true
    Write Material Random=true
   )]
  particleUpdateScript [
   &ParticleState(Kill Particles When Lifetime Has Elapsed=false, Let Infinitely Lived Particles Die When Emitter Deactivates=true)
   &ScaleSpriteSize(
    Scale Sprite Size Mode=::NewEnumerator0
    Uniform Scale Factor=&Waveform("Waveform [1]"=::NewEnumerator12, "[1] Frequency"=1, "[1] Amplitude Min/Max"=<0.8 1>)
   )
   &Light_Attributes(
    Color=1
    Falloff Exponent=1
    Light Exponent=4
    Light Radius=50
    Radius=1
    Scale Alpha:NiagaraFloat=&Multiply_Float(A=500, B=&Waveform("Waveform [1]"=::NewEnumerator12, "[1] Frequency"=1, "[1] Amplitude Min/Max"=<0.8 1>))
    ScaleA=true
    ScaleRGB=true
   )
   /SetVariables(Particles.VisibilityTag=&MakeCustomIntFromBool(Bool=$User.bisHeadHidden, True INT=1, False INT=0)) {isSetParameters=true}
  ]
  rendererClasses [&NiagaraSpriteRendererProperties, &NiagaraLightRendererProperties]
  renderers [
   {rendererClass=&NiagaraSpriteRendererProperties}
   {rendererClass=&NiagaraLightRendererProperties, rendererIndex=1}
  ]
 }}
summary {
 systemName=NS_OrnamentalSoldier_Hat
 emitters [{emitterName=Minimal, rendererClasses [&NiagaraSpriteRendererProperties, &NiagaraLightRendererProperties]}]
 userVariables [User.bisHeadHidden {type=NiagaraBool}]
}
